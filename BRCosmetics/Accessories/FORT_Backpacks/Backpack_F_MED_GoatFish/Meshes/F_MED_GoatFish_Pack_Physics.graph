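# Physics asset for the Fortnite GoatFish back bling, giving the pack and its left-named tail bones capsule bodies for secondary tail motion.
# The tail base is angularly locked to the pack, and the first dynamic tail bone hangs from that base.
# The second and third dynamic bones both branch from the first rather than chaining, each joint having narrow swing limits and a zero twist limit.
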PhysicsAsset /BRCosmetics/Accessories/FORT_Backpacks/Backpack_F_MED_GoatFish/Meshes/F_MED_GoatFish_Pack_Physics
bodies {
 base_tail_l_01 [{shapeType=Capsule, length=0.505, radius=0.505, shapeName=base_tail_l_01_capsule, rotation=<rot 0 -90 -3.9315>}]
 dyn_tail_l_01 [{shapeType=Capsule, length=6, radius=2, shapeName=dyn_tail_l_01_capsule, center=<4 0 1>, rotation=<rot 66.0404 43.3669 41.7028>}]
 dyn_tail_l_02 [
  {
   shapeType=Capsule
   length=5.8893
   radius=1
   shapeName=dyn_tail_l_02_capsule
   center=<2.4502 -0.5643 0.0389>
   rotation=<rot -3.6451 -77.6118 -87.5163>
  }
  {
   shapeType=Capsule
   length=7.1554
   radius=0.8355
   shapeName=dyn_tail_l_02_capsule
   center=<1.3678 1.339 -0.0409>
   rotation=<rot -83.9309 1.3637 -5.7291>
  }
 ]
 dyn_tail_l_03 [
  {
   shapeType=Capsule
   length=4.647
   radius=1
   shapeName=dyn_tail_l_03_capsule
   center=<2.291 0.3661 -0.189>
   rotation=<rot 5.4069 -104.7433 -86.1734>
  }
  {
   shapeType=Capsule
   length=6.3781
   radius=1.1244
   shapeName=dyn_tail_l_03_capsule
   center=<1.0108 -1.7654 -0.2161>
   rotation=<rot -13.8052 -77.2736 -92.1515>
  }
 ]
 pack [
  {shapeType=Capsule, length=22.458, radius=6.8786, center=<-3.1889 -7.3198 -13.9619>, rotation=<rot 9.7094 -81.1219 71.3753>}
  {
   shapeType=Capsule
   length=5.1422
   radius=2.7339
   shapeName=pack_capsule
   center=<1.2034 -10.7455 -10.5523>
   rotation=<rot -75.3587 0.0011 11.1942>
  }
 ]
}
constraints [
 {
  bone1Name=base_tail_l_01
  bone2Name=pack
  swing1LimitDegrees=45
  swing1Motion=Locked
  swing2LimitDegrees=45
  swing2Motion=Locked
  twistLimitDegrees=45
  twistMotion=Locked
 }
 {bone1Name=dyn_tail_l_01, bone2Name=base_tail_l_01, swing1LimitDegrees=2, swing2LimitDegrees=10, twistLimitDegrees=0}
 {bone1Name=dyn_tail_l_02, bone2Name=dyn_tail_l_01, swing1LimitDegrees=10, swing2LimitDegrees=10, twistLimitDegrees=0}
 {bone1Name=dyn_tail_l_03, bone2Name=dyn_tail_l_01, swing1LimitDegrees=10, swing2LimitDegrees=8, twistLimitDegrees=0}
]
